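# Material for the Cat Eyes vehicle wheel cosmetic, rendering tinted, glossy eyes that jitter over time, with shared vehicle damage and de-rez functions applied.
# Zero-tiled UVs panned through gradient textures turn time into an animated UV offset.
# The eye uses bump-offset parallax and CustomColor tinting, with reflection-vector swirls available as a layer.
# A normal-based, fresnel-like term and reflection texture alpha shape fake highlights baked into base color.
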Material /VehicleCosmetics/Wheels/CatEyes/Materials/M_Wheel_CatEyes_FX
graph {
 properties {
  bAutomaticallySetUsageInEditor=false
  bUseMaterialAttributes=true
  bUsedWithMeshDeformer=true
  bUsedWithSkeletalMesh=true
  usedShadingModels=Default Lit
 }
 nodes [
  n1 = SetMaterialAttributes!(
   MaterialAttributes
   Base Color=NamedRerouteUsage()
   Metallic=ScalarParameter() {parameterName=Metallic}
   Roughness=ScalarParameter() {defaultValue=0.1, parameterName=Roughness}
   Normal=NamedRerouteUsage()
  ) {attributeSetTypes [
    #1
    #2
    #3
    #4
   ]}
  MaterialFunctionCall(In=MaterialFunctionCall(Material=@n1) {materialFunction=&MF_BaseVehicle_Damage}.Result) {materialFunction=&MF_VehicleDeRez_Transition}
  n2 = TextureCoordinate()
  n3 = TextureCoordinate() {uTiling=0, vTiling=0}
  n4 = Multiply(A=ScalarParameter() {defaultValue=5, parameterName=JitterAmount}, B=Time())
  n5 = Panner(Coordinate=@n3, Time=@n4) {speedX=0.3, speedY=0.1}
  n6 = Clamp(None=Multiply(A=TextureSample(UVs=@n5) {texture=/VehicleCosmetics/Wheels/_SharedTextures/FX/MowGradient01}.R) {constB=0.5})
  n7 = Panner(Coordinate=@n3, Time=@n4) {speedX=-0.3, speedY=-0.125}
  n8 = Clamp(None=Multiply(A=TextureSample(UVs=@n7) {texture=&MowGradient01}.G) {constB=0.5})
  n9 = Multiply(A=ConstantBiasScale(None=AppendVector(A=@n6, B=@n8)) {bias=-0.8, scale=0.1}, B=ScalarParameter() {defaultValue=1, parameterName=Animation})
  n10 = Constant() {r=-1.95}
  n11 = Constant() {r=-0.9}
  NamedRerouteDeclaration(None=Add(A=Add(A=@n2, B=@n9), B=AppendVector(A=@n10, B=@n11))) {name=Jitter UV, nodeColor=<rgba 1 0 0.7734 1>}
  n12 = TextureSampleParameter2D() {parameterName=Eye Normal, samplerType=SAMPLERTYPE_Normal, texture=/VehicleCosmetics/Wheels/CatEyes/Textures/T_CatEyes_N}
  n13 = DotProduct(A=Multiply(A=Constant3Vector() {constant=<rgba 0.05 0.05 1 0>}, B=@n12.RGB), B=CameraVectorWS())
  n14 = OneMinus(None=Clamp(None=Multiply(A=@n13) {constB=5}) {maxDefault=1.1})
  n15 = Multiply(A=@n12.RGB, B=Constant3Vector() {constant=<rgba 0.66 0.66 1 0>})
  n16 = ComponentMask(None=Transform(None=@n15)) {b=true}
  n17 = Multiply(A=Add(A=NamedRerouteUsage(), B=Multiply(A=@n14, B=@n14)), B=Clamp(None=ConstantBiasScale(None=@n16)))
  n18 = TextureSample(Tex=NamedRerouteUsage()) {samplerSource=SSM_Wrap_WorldGroupSettings, texture=/VehicleCosmetics/Wheels/CatEyes/Textures/T_CatEyes_Reflection}
  n19 = Multiply(A=@n18.A, B=@n18.A)
  NamedRerouteDeclaration(None=LinearInterpolate(A=NamedRerouteUsage(), B=Multiply(A=@n17, B=@n19), Alpha=OneMinus(None=@n19))) {name=BaseColor, nodeColor=<rgba 0 1 0.9063 1>}
  NamedRerouteDeclaration(None=@n15) {name=Normal, nodeColor=<rgba 0.5 0.5 1 1>}
  n20 = ComponentMask(None=Transform(None=ReflectionVectorWS())) {g=true, r=true}
  n21 = Constant2Vector() {g=0.3, r=0.3}
  NamedRerouteDeclaration(None=TextureSample(UVs=Multiply(A=@n20, B=@n21)) {texture=/VehicleCosmetics/Wheels/_SharedTextures/FX/Swirls_D}.RGB) {name=Swirls, nodeColor=<rgba 0.875 1 0 1>}
  n22 = VectorParameter() {group=0_Base, parameterName=CustomColor, sortPriority=0, defaultValue=<rgba 0.9823 0.7991 0.0648 1>}
  n23 = NamedRerouteUsage()
  n24 = TextureObjectParameter() {parameterName=Eye, texture=/VehicleCosmetics/Wheels/CatEyes/Textures/T_CatEyes_BC}
  n25 = TextureSample(
   UVs=BumpOffset(
    Coordinate=@n23
    Height=TextureSample(UVs=@n23, Tex=@n24) {samplerSource=SSM_Wrap_WorldGroupSettings, texture=&T_CatEyes_BC}.A
   ) {heightRatio=0.25, referencePlane=0.4}
   Tex=@n24
  ) {samplerSource=SSM_Wrap_WorldGroupSettings, texture=&T_CatEyes_BC}
  n26 = Multiply(A=@n22.RGB, B=@n25.RGB) {constB=2}
  n27 = Multiply(A=@n26, B=NamedRerouteUsage())
  NamedRerouteDeclaration(None=LinearInterpolate(A=Add(A=@n26, B=@n27), B=@n22.RGB, Alpha=Multiply(A=@n25.G) {constB=2})) {name=Eye, desc=Might go straightaway to BC, nodeColor=<rgba 0.8047 0 1 1>}
  NamedRerouteDeclaration(None=@n24) {name=Eye tex, nodeColor=<rgba 0.2187 1 0 1>}
 ]
}
